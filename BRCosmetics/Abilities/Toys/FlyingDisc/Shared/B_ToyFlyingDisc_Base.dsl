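; Base Blueprint for the throwable flying disc toy, a spinning physics projectile that players can catch.
; On throw, timelines bend velocity by launch tilt to curve, slow or return the disc, until impacts stop them and add randomized bounces.
; Pawn catches before the disc is spent trigger scoring effects, destruction and replicated pass-distance messages, with bounce counts for futuristic variants.
; A delayed gate permits self-catches, which report quest progress.
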(event EventProjectileStop
  (params (ImpactResult :type Hit%20Result%20Structure%20%28by%20ref%29))
  (bind _flyingdiscmesh (Variables|Default|GetFlyingDiscMesh))
  ;[1 1 1] Rotate Mesh Flat To Ground
  (Utilities|FlowControl|DoOnce false
    (:Completed
      (Collision|SetCollisionEnabled _flyingdiscmesh "PhysicsOnly")
      (Physics|SetSimulatePhysics _flyingdiscmesh true))))

(event Visuals|EventInitializeToyPreviewInFrontend
  (|CallParentFunction)
  (Components|Activation|SetActive (Variables|Default|GetFootballTrail) false false))

(event Custom|PlayScoringAnim
  (bind (_returnvalue_location _returnvalue_rotation) (Math|Transform|BreakTransform (Transformation|GetWorldTransform (Variables|FortPhysicsBall|GetSphereCollision))))
  ;[1 1 1] Celebration FX
  (Variables|/Script/Engine.ProjectileMovementComponent|SetbSimulationEnabled false (Variables|Default|GetProjectileMovement))
  (Collision|SetActorEnableCollision false)
  (Utilities|FlowControl|Sequence
    (:then_0
      (Networking|SwitchHasAuthority
        (:Authority
          (Utilities|FlowControl|Delay (Variables|Default|GetLerpDuration))
          (Gameplay|SilentDie false))))
    (:then_1
      (Components|Timeline|SetPlayRate (Variables|B_ToyFlyingDisc_Base|GetLerpIntoTouchdown) (/ 1.0 (Variables|Default|GetLerpDuration)))
      (bind (_direction _position _scale) (Timeline|LerpIntoTouchdown PlayFromStart)
        (:Update
          (Transformation|SetWorldTransform (Variables|FortPhysicsBall|GetSphereCollision) (Math|Transform|MakeTransform (Math|Vector|Lerp(Vector) (Variables|Default|GetIdealLerpStartPos) (Variables|Default|GetIdealLerpEndPos) _position) _returnvalue_rotation (Math|Conversions|ToVector(Float) _scale)) false false))))))

(event OnProjectileBounce(ProjectileMovement)
  (params (ImpactResult :type Hit%20Result%20Structure%20%28by%20ref%29) (ImpactVelocity :type Vector%20%28by%20ref%29 :default "0, 0, 0"))
  (bind _projectilemovement (Variables|Default|GetProjectileMovement))
  (bind _velocity (Variables|/Script/Engine.MovementComponent|GetVelocity _projectilemovement))
  (bind _returnvalue (Math|Vector|VectorLength _velocity))
  (Utilities|FlowControl|Sequence
    (:then_0
      (if (Variables|Default|GetIsFuturistic)
        (Math|Integer|IncrementInt (Variables|Default|GetBounces))))
    (:then_1
      ;[1 1 1] On hit - give a random impulse based on the hit normal
      (Utilities|FlowControl|Sequence
        (:then_0
          (CallFunction|StopTimelines)
          (if (not (Variables|Default|GetBounce))
            (Variables|Default|SetBounce true)
            (if (>= _returnvalue 15.0)
              (Variables|/Script/Engine.MovementComponent|SetVelocity (Math|Vector|ClampVectorSize (* (Math|Vector|RotateVector (stale Collision|EventHit HitNormal) (Math|Rotator|MakeRotator (Math|Random|RandomFloatinRange -25.0 25.0) (Math|Random|RandomFloatinRange -25.0 25.0) (Math|Random|RandomFloatinRange -25.0 25.0))) _returnvalue) 0.0 9e+02) _projectilemovement))
            (Utilities|FlowControl|DoOnce false
              (:Completed
                (Variables|Default|SetHit true) ; Hit Variable
                (Components|Activation|SetActive (Variables|Default|GetFootballTrail) true true)
                (Effects|Particles|Trails|EndTrails (Variables|Default|GetFootballTrail))
                (Variables|/Script/Engine.ProjectileMovementComponent|SetProjectileGravityScale 0.7 (Variables|Default|GetProjectileMovement))))))
        (:then_1
          (Utilities|FlowControl|DoOnce false
            (:Completed
              (Audio|Components|Audio|FadeOut (Variables|Default|GetInAirAudio) 0.25 0.0 "Linear" :!class /Script/Engine.AudioComponent))))
        (:then_2
          (Utilities|FlowControl|DoOnce false
            (:Completed
              (Effects|SpawnEmitteratLocation (Variables|Default|GetImpactEffect) (stale Collision|EventHit HitLocation) (Math|Conversions|RotationFromXVector (stale Collision|EventHit HitNormal))))))))))

(event Collision|EventHit
  (params (MyComp :type Primitive%20Component%20Object%20Reference) (Other :type Actor%20Object%20Reference) (OtherComp :type Primitive%20Component%20Object%20Reference) (bSelfMoved :type Boolean) (HitLocation :type Vector :default "0, 0, 0") (HitNormal :type Vector :default "0, 0, 0") (NormalImpulse :type Vector :default "0, 0, 0") (Hit :type Hit%20Result%20Structure%20%28by%20ref%29))
  (bind _projectilemovement (Variables|Default|GetProjectileMovement))
  (bind _velocity (Variables|/Script/Engine.MovementComponent|GetVelocity _projectilemovement))
  (bind _returnvalue (Math|Vector|VectorLength _velocity))
  ;[1 1 1] On hit - give a random impulse based on the hit normal
  (|CallParentFunction MyComp Other OtherComp bSelfMoved HitLocation HitNormal NormalImpulse Hit)
  (Utilities|FlowControl|Sequence
    (:then_0
      (CallFunction|StopTimelines)
      (if (not (Variables|Default|GetBounce))
        (Variables|Default|SetBounce true)
        (if (>= _returnvalue 15.0)
          (Variables|/Script/Engine.MovementComponent|SetVelocity (Math|Vector|ClampVectorSize (* (Math|Vector|RotateVector HitNormal (Math|Rotator|MakeRotator (Math|Random|RandomFloatinRange -25.0 25.0) (Math|Random|RandomFloatinRange -25.0 25.0) (Math|Random|RandomFloatinRange -25.0 25.0))) _returnvalue) 0.0 9e+02) _projectilemovement))
        (Utilities|FlowControl|DoOnce false
          (:Completed
            (Variables|Default|SetHit true) ; Hit Variable
            (Components|Activation|SetActive (Variables|Default|GetFootballTrail) true true)
            (Effects|Particles|Trails|EndTrails (Variables|Default|GetFootballTrail))
            (Variables|/Script/Engine.ProjectileMovementComponent|SetProjectileGravityScale 0.7 (Variables|Default|GetProjectileMovement))))))
    (:then_1
      (Utilities|FlowControl|DoOnce false
        (:Completed
          (Audio|Components|Audio|FadeOut (Variables|Default|GetInAirAudio) 0.25 0.0 "Linear" :!class /Script/Engine.AudioComponent))))
    (:then_2
      (Utilities|FlowControl|DoOnce false
        (:Completed
          (Effects|SpawnEmitteratLocation (Variables|Default|GetImpactEffect) HitLocation (Math|Conversions|RotationFromXVector HitNormal)))))))

(event Collision|EventActorBeginOverlap
  (params (OtherActor :type Actor%20Object%20Reference))
  (|CallParentFunction OtherActor)
  (CallFunction|HitPlayer/VehicleBounceSound))

(event Custom|StartLookRotationLocation
  (params (Location :type Vector) (Rotation :type Rotator))
  (Variables|Default|SetStartLookRotation Rotation)
  (Variables|Default|SetPlayerThrownLocation (+ (+ Location (* (Math|Vector|GetForwardVector (Math|Rotator|MakeRotator 0.0 0.0 (.yaw Rotation))) 4.5e+03)) "0, 0,-200.000000")))

(event Visuals|EventInitializeToyInstance
  (params (OwningPC :type Fort%20Player%20Controller%20Object%20Reference) (NumTimesSummoned :type Integer))
  (|CallParentFunction OwningPC NumTimesSummoned)
  (Variables|Default|SetHasFinishedSpawning true))

(event Custom|OpenSelfCatchGate
  (bind _returnvalue (Actor|GetOwner))
  (bind _asfort_player_state (Utilities|Casting|CastToFortPlayerState _returnvalue))
  ;[1 1 1] Call celebration FX and message when a pawn is hit
  (Utilities|FlowControl|Delay 0.4)
  (Utilities|FlowControl|Gate true
    (:Exit
      (bind _meterstraveled (CallFunction|CalcDistanceTraveled (Transformation|GetActorLocation (stale EventOnPawnTouchedMe TouchingPawn))))
      (CallFunction|CreateMessage _asfort_player_state (stale EventOnPawnTouchedMe _asfort_player_state_2) _meterstraveled (Variables|Default|GetBounces))
      (CallFunction|QuestProgressSelfCaught (Utilities|Equal(Object) _asfort_player_state (stale EventOnPawnTouchedMe _asfort_player_state_2)))
      (Networking|SwitchHasAuthority
        (:Authority
          ;;[1 0.8413 0.08369] Play FX and kill on server
          (CallFunction|PlayScoringCueandDestruct (Transformation|GetActorLocation (stale EventOnPawnTouchedMe TouchingPawn)) (+ (Transformation|GetActorLocation (stale EventOnPawnTouchedMe TouchingPawn)) (* (Transformation|GetActorUpVector) -1.2e+02)))
          (Gameplay|SilentDie false))
        (:Remote
          ;;[0.1583 1 0.2487] Hide on client
          (Development|SetHiddeninGame (Variables|Default|GetBall) true false))))))

(event EventBeginPlay
  (|CallParentFunction)
  (Variables|Default|SetIs%20Throwing true))

(event EventTick
  (params (DeltaSeconds :type Float%20%28single-precision%29))
  (if (Variables|Default|GetHit)
    (|CallParentFunction DeltaSeconds) ; Becomes like any other toy after hit
    (else
      (Networking|CanExecuteCosmeticEvents
        (:True
          ;[1 1 1] Manually handle rotation until we hit something (this is the spiral)
          (Transformation|AddLocalRotation (Variables|Default|GetFlyingDiscMesh) "0.000000,40.000000, 0" false false))))))

(event EventOnPawnTouchedMe
  (params (TouchingPawn :type Pawn%20Object%20Reference) (ImpactNormal :type Vector :default "0, 0, 0") (OtherVelocity :type Vector :default "0, 0, 0"))
  (bind _returnvalue (Actor|GetOwner))
  (bind _asfort_player_state (Utilities|Casting|CastToFortPlayerState _returnvalue))
  ;[1 1 1] Call celebration FX and message when a pawn is hit
  (|CallParentFunction TouchingPawn ImpactNormal OtherVelocity)
  (bind _asfort_player_pawn (Utilities|Casting|CastToFortPlayerPawn TouchingPawn))
  (bind _asfort_player_state_2 (Utilities|Casting|CastToFortPlayerState (Variables|/Script/Engine.Controller|GetPlayerState (Pawn|GetController _asfort_player_pawn))))
  (if (and (not (Variables|Default|GetHit)) (Variables|Default|GetHasFinishedSpawning))
    (if (Utilities|Equal(Object) _asfort_player_state_2 (Actor|GetOwner))
      (Utilities|FlowControl|Gate true
        (:Exit
          _)))
    (bind _meterstraveled (CallFunction|CalcDistanceTraveled (Transformation|GetActorLocation TouchingPawn)))
    (CallFunction|CreateMessage _asfort_player_state _asfort_player_state_2 _meterstraveled (Variables|Default|GetBounces))
    (CallFunction|QuestProgressSelfCaught (Utilities|Equal(Object) _asfort_player_state _asfort_player_state_2))
    (Networking|SwitchHasAuthority
      (:Authority
        ;;[1 0.8413 0.08369] Play FX and kill on server
        (CallFunction|PlayScoringCueandDestruct (Transformation|GetActorLocation TouchingPawn) (+ (Transformation|GetActorLocation TouchingPawn) (* (Transformation|GetActorUpVector) -1.2e+02)))
        (Gameplay|SilentDie false))
      (:Remote
        ;;[0.1583 1 0.2487] Hide on client
        (Development|SetHiddeninGame (Variables|Default|GetBall) true false)))))

(event EventSetBallVelocityAndStartTicking
  (params (NewVelocity :type Vector :default "0, 0, 0"))
  (|CallParentFunction NewVelocity)
  (Variables|Default|SetStartingVelocity NewVelocity)
  (CallFunction|ClientonThrown NewVelocity))

(event Custom|Client%20On%20Thrown
  (params (StartingVelocity :type Vector))
  (bind _startlookrotation (Variables|Default|GetStartLookRotation))
  (bind _yaw (.yaw _startlookrotation))
  (bind _returnvalue (Math|Rotator|MakeRotator 0.0 0.0 _yaw))
  (bind _returnvalue_2 (Math|Rotator|GetRotationXVector _returnvalue))
  (bind _returnvalue_3 (Math|Vector|Normalize StartingVelocity))
  (bind _returnvalue_4 (Math|Rotator|MakeRotfromZ _returnvalue_3))
  (bind _returnvalue_5 (Math|Rotator|GetRotationXVector _returnvalue_4))
  (bind _startlookrotation_2 (Variables|Default|GetStartLookRotation))
  (bind _yaw_2 (.yaw _startlookrotation_2))
  (bind _returnvalue_6 (+ _yaw_2 9e+01))
  (bind _returnvalue_7 (Math|Rotator|MakeRotator 0.0 0.0 _returnvalue_6))
  (bind _returnvalue_8 (Math|Rotator|GetRotationXVector _returnvalue_7))
  (bind _returnvalue_9 (Math|Rotator|MakeRotfromZ _returnvalue_3))
  (bind _returnvalue_10 (Math|Rotator|GetRotationXVector _returnvalue_9))
  (bind _returnvalue_11 (Math|Vector|DotProduct _returnvalue_10 _returnvalue_8))
  (bind _returnvalue_12 (Math|Vector|DotProduct _returnvalue_2 _returnvalue_5))
  (bind _returnvalue_13 (Math|Trig|Acos(Degrees) _returnvalue_12))
  (bind _returnvalue_14 (> _returnvalue_13 9e+01))
  (bind _returnvalue_15 (Math|Trig|Acos(Degrees) _returnvalue_11))
  (bind _returnvalue_16 (- _returnvalue_13 9e+01))
  (bind _returnvalue_17 (Math|Float|Clamp(Float) _returnvalue_16 -35.0 35.0))
  (bind _returnvalue_18 (* _returnvalue_17 -1.2))
  (bind _returnvalue_19 (* _returnvalue_17 -1.2))
  (bind _returnvalue_20 (Math|Float|SelectFloat _returnvalue_19 _returnvalue_18 _returnvalue_14))
  (bind _returnvalue_21 (Actor|GetOwner))
  (bind _asfort_player_state (Utilities|Casting|CastToFortPlayerState _returnvalue_21))
  (bind _returnvalue_22 (PlayerState|GetCurrentPawn _asfort_player_state))
  (bind (_outlocation _outrotation) (Actor|GetActorEyesViewPoint _returnvalue_22))
  (bind _roll (.roll _outrotation))
  (bind _pitch (.pitch _outrotation))
  (bind _yaw_3 (.yaw _outrotation))
  (bind _returnvalue_23 (Math|Rotator|MakeRotator _roll _pitch _yaw_3))
  (bind _returnvalue_24 (Math|Vector|GetForwardVector _returnvalue_23))
  (bind _returnvalue_25 (Transformation|GetActorForwardVector _returnvalue_22))
  (bind _returnvalue_26 (Math|Vector|DotProduct _returnvalue_25 _returnvalue_24))
  (bind _returnvalue_27 (Math|Trig|Acos(Degrees) _returnvalue_26))
  (bind _returnvalue_28 (> _returnvalue_27 3e+01))
  ;[1 1 1] Curve Flying Disc
  (Utilities|FlowControl|Sequence
    (:then_0
      (Variables|Default|SetIs%20Throwing false)
      (Audio|Components|Audio|FadeIn (Variables|Default|GetInAirAudio) 0.25 1.0 0.0 "Linear" :!class /Script/Engine.AudioComponent)
      (if (> (.z (+ (Transformation|GetActorLocation) (* (Math|Vector|GetForwardVector (Variables|Default|GetStartLookRotation)) 5e+02))) (.z (Variables|Default|GetPlayerThrownLocation)))
        (if (> _returnvalue_15 8e+01)
          (Development|PrintString "NoTilting" true true "(R=0.000000,G=0.660000,B=1.000000,A=1.000000)" 5.0 :!label _label)
          (if (not _returnvalue_28)
            (Utilities|FlowControl|Delay 0.5)
            (Variables|Default|SetStartingVelocity (Variables|/Script/Engine.MovementComponent|GetVelocity (Variables|Default|GetProjectileMovement)))
            (Utilities|FlowControl|Gate false
              (:Exit
                (bind (_direction _newtrack_0) (Timeline|Slow%20Speed PlayFromStart)
                  (:Update
                    (Variables|/Script/Engine.MovementComponent|SetVelocity (Math|Vector|MakeVector (.x (Math|Vector|Lerp(Vector) (Variables|Default|GetStartingVelocity) (* (Variables|Default|GetStartingVelocity) 0.5) _newtrack_0)) (.y (Math|Vector|Lerp(Vector) (Variables|Default|GetStartingVelocity) (* (Variables|Default|GetStartingVelocity) 0.5) _newtrack_0)) (.z (Variables|/Script/Engine.MovementComponent|GetVelocity (Variables|Default|GetProjectileMovement)))) (Variables|Default|GetProjectileMovement)))))))
          (elif (> _returnvalue_15 7.0)
            (Transformation|SetRelativeRotation (Variables|Default|GetBall) (Math|Rotator|MakeRotator (+ _returnvalue_20 1.8e+02)) false false)
            (if _returnvalue_28
              (if (not (> (Math|Float|Absolute(Float) _returnvalue_20) 25.0))
                (Variables|Default|SetPlayerThrownLocation (Transformation|GetActorLocation (PlayerState|GetCurrentPawn (Utilities|Casting|CastToFortPlayerState (Actor|GetOwner)))) :!label _label_2)
                (Networking|SwitchHasAuthority
                  (:Authority
                    (Utilities|FlowControl|Delay 0.2)
                    (Variables|Default|SetStartingVelocity (Variables|/Script/Engine.MovementComponent|GetVelocity (Variables|Default|GetProjectileMovement)))
                    (Utilities|FlowControl|Gate false
                      (:Exit
                        ;;[1 1 1] Reverse Back Towards Player When Thrown Too High
                        (bind (_direction_2 _newtrack_0_2) (Timeline|Timeline_1 PlayFromStart)
                          (:Update
                            _)
                          (:Finished
                            _))
                        (Variables|/Script/Engine.MovementComponent|SetVelocity (Math|Vector|MakeVector (.x (Math|Vector|Lerp(Vector) (Variables|Default|GetStartingVelocity) (* (Math|Vector|Lerp(Vector) (Math|Vector|Normalize (- (Variables|Default|GetPlayerThrownLocation) (Transformation|GetActorLocation))) "0, 0,-0.500000" 0.25) 1.5e+03) _newtrack_0_2)) (.y (Math|Vector|Lerp(Vector) (Variables|Default|GetStartingVelocity) (* (Math|Vector|Lerp(Vector) (Math|Vector|Normalize (- (Variables|Default|GetPlayerThrownLocation) (Transformation|GetActorLocation))) "0, 0,-0.500000" 0.25) 1.5e+03) _newtrack_0_2)) (.z (Math|Vector|Lerp(Vector) (Variables|Default|GetStartingVelocity) (* (Math|Vector|Lerp(Vector) (Math|Vector|Normalize (- (Variables|Default|GetPlayerThrownLocation) (Transformation|GetActorLocation))) "0, 0,-0.500000" 0.25) 1.5e+03) _newtrack_0_2))) (Variables|Default|GetProjectileMovement))))))))
            (Networking|SwitchHasAuthority
              (:Authority
                (Utilities|FlowControl|Delay 0.45)
                (Variables|Default|SetStartingVelocity (Variables|/Script/Engine.MovementComponent|GetVelocity (Variables|Default|GetProjectileMovement)))
                (Utilities|FlowControl|Gate false
                  (:Exit
                    (bind (_direction_3 _newtrack_0_3) (Timeline|Timeline_0 PlayFromStart)
                      (:Update
                        (Variables|/Script/Engine.MovementComponent|SetVelocity (Math|Vector|MakeVector (.x (Math|Vector|Lerp(Vector) (Variables|Default|GetStartingVelocity) (* (Math|Vector|RotateVectorAroundAxis (Transformation|GetActorRightVector) (* _returnvalue_20 -1.0) (Transformation|GetActorForwardVector)) (* (Math|Float|SelectFloat (* (Variables|Default|GetTiltSpeed) -1.0) (Variables|Default|GetTiltSpeed) _returnvalue_14) (Math|Float|Clamp(Float) (/ (Math|Float|Absolute(Float) _returnvalue_16) 55.0)))) _newtrack_0_3)) (.y (Math|Vector|Lerp(Vector) (Variables|Default|GetStartingVelocity) (* (Math|Vector|RotateVectorAroundAxis (Transformation|GetActorRightVector) (* _returnvalue_20 -1.0) (Transformation|GetActorForwardVector)) (* (Math|Float|SelectFloat (* (Variables|Default|GetTiltSpeed) -1.0) (Variables|Default|GetTiltSpeed) _returnvalue_14) (Math|Float|Clamp(Float) (/ (Math|Float|Absolute(Float) _returnvalue_16) 55.0)))) _newtrack_0_3)) (.z (Math|Vector|Lerp(Vector) (Variables|Default|GetStartingVelocity) (* (Math|Vector|RotateVectorAroundAxis (Transformation|GetActorRightVector) (* _returnvalue_20 -1.0) (Transformation|GetActorForwardVector)) (* (Math|Float|SelectFloat (* (Variables|Default|GetTiltSpeed) -1.0) (Variables|Default|GetTiltSpeed) _returnvalue_14) (Math|Float|Clamp(Float) (/ (Math|Float|Absolute(Float) _returnvalue_16) 55.0)))) _newtrack_0_3))) (Variables|Default|GetProjectileMovement))))))))
            (else
              (goto _label))))
        (goto _label_2)
        (else
          (Development|PrintString "NotHighEnough" true true "(R=0.000000,G=0.660000,B=1.000000,A=1.000000)" 5.0))))
    (:then_1
      (Utilities|FlowControl|Gate false
        (:Exit
          (bind (_direction_4 _newtrack_0_4) (Timeline|Timeline_2 PlayFromStart)
            (:Update
              (Variables|/Script/Engine.ProjectileMovementComponent|SetProjectileGravityScale (Math|Float|Lerp 0.025 0.05 _newtrack_0_4) (Variables|Default|GetProjectileMovement)))))))
    (:then_2
      (Components|Activation|Activate (Variables|Default|GetFootballTrail) false)
      (Effects|Particles|Trails|BeginTrails (Variables|Default|GetFootballTrail) "FX_r1" "FX_l1" "ETrailWidthMode_FromCentre" 1.5)
      (Development|PrintString "Hello"))
    (:then_3
      (CallFunction|OpenSelfCatchGate))))

(event Custom|StopTimelines
  ;[1 1 1] Curve Flying Disc
  (Utilities|FlowControl|Sequence
    (:then_0
      ;;[1 1 1] Reverse Back Towards Player When Thrown Too High
      (Timeline|Timeline_1 Stop))
    (:then_1
      (Timeline|Timeline_0 Stop))
    (:then_2
      (Timeline|Slow%20Speed Stop))
    (:then_3
      (Timeline|Timeline_2 Stop))
    (:then_4
      (Utilities|FlowControl|Gate false
        (:Exit
          (Timeline|Timeline_1 PlayFromStart))))
    (:then_5
      (Utilities|FlowControl|Gate false
        (:Exit
          (Timeline|Timeline_0 PlayFromStart))))
    (:then_6
      (Utilities|FlowControl|Gate false
        (:Exit
          (Timeline|Slow%20Speed PlayFromStart))))
    (:then_7
      (Utilities|FlowControl|Gate false
        (:Exit
          (Timeline|Timeline_2 PlayFromStart))))))

(event Custom|Hit%20Player/Vehicle%20Bounce%20Sound
  ;[1 1 1] Support Collision with Players and Vehicles
  (if (not (Variables|Default|GetIs%20Throwing))
    (Utilities|FlowControl|DoOnce false :!label _label
      (:Completed
        (Audio|PlaySoundatLocation "/BRCosmetics/Sounds/Other/Toys/FlyingDisc/Toy_FlyingDisc_Bounce_Cue.Toy_FlyingDisc_Bounce_Cue" (Transformation|GetActorLocation))
        (Utilities|FlowControl|Delay 0.2)
        (goto _label Reset)))))

(fn CalcDistanceTraveled
  (params (HitLocation :type Vector))
  (bind _output_get (Variables|Default|SetMetersTraveled (/ (Math|Vector|VectorLengthXY (- HitLocation (Variables|Default|GetInitialSpawnPosition))) 1e+02)))
  (return _output_get))

(fn CreateMessage
  (params (PlayerWhoThrew :type Fort%20Player%20State%20Object%20Reference) (PlayerWhoCaught :type Fort%20Player%20State%20Object%20Reference) (PassDistance :type Float%20%28double-precision%29) (Bounces :type Integer))
  (if (Variables|/Script/FortniteGame.FortRuntimeOptions|GetbToyMessagingEnabled (Options|GetRuntimeOptions))
    (if (Variables|Default|GetIsFuturistic)
      (if (> Bounces 0)
        (Variables|Default|SetDisplayFuturisticMessage (Utilities|Struct|MakeStructFlyingDiscFuturisticMinigameMessage PlayerWhoThrew PlayerWhoCaught PassDistance (Variables|Default|GetMessageSN) Bounces))
        (else
          (Variables|Default|SetDisplayMessage (Utilities|Struct|MakeStructFlyingDiscMinigameMessage PlayerWhoThrew PlayerWhoCaught PassDistance (Variables|Default|GetMessageSN)) :!label _label)))
      (Variables|Default|SetMessageSN (Math|Byte|Byte+Byte (Variables|Default|GetMessageSN) 1)))
    (goto _label)))

(fn OnRep_DisplayFuturisticMessage
  (bind _displayfuturisticmessage (Variables|Default|GetDisplayFuturisticMessage))
  (bind (_playerwhothrew_22_ef5d780b4020e911ab429d9ba6b977ba _playerwhocaught_25_800cc3c44dd84e656fcddc9e415ec97d _passdistance_26_a600d2a048cb01620b914bb4107a6cd5 _serialnumber_20_3cc5eca54decc5eae3e06394e71de575 _bounces_29_ad5c69904af03c7507c40ca6771e14f7) (Utilities|Struct|BreakStructFlyingDiscFuturisticMinigameMessage _displayfuturisticmessage))
  (bind _returnvalue (BlueprintContext|GetFortToyContext))
  (bind _returnvalue_2 (BlueprintContext|GetFortToyContext))
  (Networking|SwitchHasAuthority
    (:Remote
      (Utilities|IsValid _playerwhothrew_22_ef5d780b4020e911ab429d9ba6b977ba
        (:Is%20Valid
          (bind _returnvalue_3 (Class|FortToyContext|FormatPlayerNameforDisplay _returnvalue _playerwhothrew_22_ef5d780b4020e911ab429d9ba6b977ba "ToyOwner"))
          (bind _returnvalue_4 (Class|FortToyContext|FormatPlayerNameforDisplay _returnvalue _playerwhocaught_25_800cc3c44dd84e656fcddc9e415ec97d "OtherPlayerInteractingWithToy"))
          (CallFunction|PickFormatString)
          (Utilities|IsValid _returnvalue_2
            (:Is%20Valid
              (Class|FortToyContext|ConditionalPrinttoMessageFeed _returnvalue_2 _playerwhothrew_22_ef5d780b4020e911ab429d9ba6b977ba (Transformation|GetActorLocation) (select (> _passdistance_26_a600d2a048cb01620b914bb4107a6cd5 (Variables|Default|GetMinimumDistanceforTeamRelevance)) "SquadAndTeamMembers" "SquadMembersOnly") (Utilities|Text|FormatText "NSLOCTEXT(\"[9C5B643E4D047542B1856C874B948C9F]\", \"9FF84AC1462AAF5C743D80AC5406180B\", \"{ThrowingPlayer} threw a {Distance} meter pass to {CatchingPlayer} after it richocheted {Bounces} times!\")" _returnvalue_3 (Math|Float|Round _passdistance_26_a600d2a048cb01620b914bb4107a6cd5) _returnvalue_4 _bounces_29_ad5c69904af03c7507c40ca6771e14f7)))))))))

(fn OnRep_IdealLerpEndPos)

(fn PickFormatString
  (bind (_displaymessage_playerwhothrew_22_ef5d780b4020e911ab429d9ba6b977ba _displaymessage_playerwhocaught_25_800cc3c44dd84e656fcddc9e415ec97d _displaymessage_passdistance_26_a600d2a048cb01620b914bb4107a6cd5) (Utilities|Struct|BreakStructFlyingDiscMinigameMessage (Variables|GetDisplayMessage_PlayerWhoThrew_22_EF5D780B4020E911AB429D9BA6B977BA)))
  (if (<= _displaymessage_passdistance_26_a600d2a048cb01620b914bb4107a6cd5 9.9)
    (return (Variables|Default|GetMinimumDistanceMessage))
    (elif (Math|Float|InRange(Float) _displaymessage_passdistance_26_a600d2a048cb01620b914bb4107a6cd5 1e+01 2e+01)
      (return (Variables|Default|GetFirstDownMessage))
      (elif (Math|Float|InRange(Float) _displaymessage_passdistance_26_a600d2a048cb01620b914bb4107a6cd5 2e+01 4e+01)
        (return (Variables|Default|GetTouchdownMessage))
        (elif (>= _displaymessage_passdistance_26_a600d2a048cb01620b914bb4107a6cd5 4e+01)
          (return (Variables|Default|GetHailMaryMessage)))))))

(fn PlayScoringCueAndDestruct
  (params (Target%20Start%20Pos :type Vector) (Target%20End%20Pos :type Vector))
  (CallFunction|PlayCueCenteredonMe (Variables|Default|GetScoreCue))
  (Variables|Default|SetIdealLerpStartPos Target%20Start%20Pos)
  (Variables|Default|SetIdealLerpEndPos Target%20End%20Pos))

(fn QuestProgress_SelfCaught
  (params (SelfCaught :type Boolean))
  (if (and SelfCaught (Networking|IsServer))
    (Quest|SendComplexCustomStatEvent (Quests|GetQuestManager (Fortnite|GetFortPlayerControllerfromActor (Actor|GetOwner)) "Invalid") 0 "(GameplayTags=((TagName=\"Athena.Quests.FlyingDisc.SelfCatch\")))" "(GameplayTags=)" 1)))

(fn ConstructionScript
  (|CallParentFunction))
