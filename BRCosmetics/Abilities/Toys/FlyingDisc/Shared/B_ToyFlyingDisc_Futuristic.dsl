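; Futuristic flying disc toy Blueprint that celebrates and ends its flight when a pawn touches it.
; The touch handler checks hit and spawn state and compares the pawn with the thrower, whose own catch goes through a gate tied to a short delay.
; Scoring creates a message with distance traveled and bounces.
; The server plays the score cue, sets lerp positions and silently kills the disc, while clients hide the mesh.
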(event EventBeginPlay
  (|CallParentFunction) ; This node is disabled and will not be called.\nDrag off pins to build functionality.
)

(event Collision|EventActorBeginOverlap
  (params (OtherActor :type Actor%20Object%20Reference))
  (|CallParentFunction OtherActor) ; This node is disabled and will not be called.\nDrag off pins to build functionality.
)

(event EventTick
  (params (DeltaSeconds :type Float%20%28single-precision%29))
  (|CallParentFunction DeltaSeconds) ; This node is disabled and will not be called.\nDrag off pins to build functionality.
)

(event Custom|CustomEvent
  ;[1 1 1] Call celebration FX and message when a pawn is hit
  (Utilities|FlowControl|Delay 0.4)
  (Utilities|FlowControl|Gate true
    (:Exit
      (bind _meterstraveled (CallFunction|CalcDistanceTraveled (Transformation|GetActorLocation (stale EventOnPawnTouchedMe TouchingPawn))))
      (CallFunction|CreateMessage (Utilities|Casting|CastToFortPlayerState (Actor|GetOwner)) (stale EventOnPawnTouchedMe _asfort_player_state) _meterstraveled (Variables|Default|GetBounces))
      (Networking|SwitchHasAuthority
        (:Authority
          ;;[1 0.8413 0.08369] Play FX and kill on server
          (CallFunction|PlayCueCenteredonMe (Variables|Default|GetScoreCue))
          (Variables|Default|SetIdealLerpStartPos (Transformation|GetActorLocation (stale EventOnPawnTouchedMe TouchingPawn)))
          (Variables|Default|SetIdealLerpEndPos (+ (Transformation|GetActorLocation (stale EventOnPawnTouchedMe TouchingPawn)) (* (Transformation|GetActorUpVector) -1.2e+02)))
          (Gameplay|SilentDie false))
        (:Remote
          ;;[0.1583 1 0.2487] Hide on client
          (Development|SetHiddeninGame (Variables|Default|GetBall) true false))))))

(event EventOnPawnTouchedMe
  (params (TouchingPawn :type Pawn%20Object%20Reference) (ImpactNormal :type Vector :default "0, 0, 0") (OtherVelocity :type Vector :default "0, 0, 0"))
  (bind _returnvalue (Transformation|GetActorLocation TouchingPawn))
  ;[1 1 1] Call celebration FX and message when a pawn is hit
  (|CallParentFunction TouchingPawn ImpactNormal OtherVelocity)
  (bind _asfort_player_pawn (Utilities|Casting|CastToFortPlayerPawn TouchingPawn))
  (bind _asfort_player_state (Utilities|Casting|CastToFortPlayerState (Variables|/Script/Engine.Controller|GetPlayerState (Pawn|GetController _asfort_player_pawn))))
  (if (and (not (Variables|Default|GetHit)) (Variables|Default|GetHasFinishedSpawning))
    (if (Utilities|Equal(Object) _asfort_player_state (Actor|GetOwner))
      (Utilities|FlowControl|Gate true
        (:Exit
          _)))
    (bind _meterstraveled (CallFunction|CalcDistanceTraveled (Transformation|GetActorLocation TouchingPawn)))
    (CallFunction|CreateMessage (Utilities|Casting|CastToFortPlayerState (Actor|GetOwner)) _asfort_player_state _meterstraveled (Variables|Default|GetBounces))
    (Networking|SwitchHasAuthority
      (:Authority
        ;;[1 0.8413 0.08369] Play FX and kill on server
        (CallFunction|PlayCueCenteredonMe (Variables|Default|GetScoreCue))
        (Variables|Default|SetIdealLerpStartPos _returnvalue)
        (Variables|Default|SetIdealLerpEndPos (+ _returnvalue (* (Transformation|GetActorUpVector) -1.2e+02)))
        (Gameplay|SilentDie false))
      (:Remote
        ;;[0.1583 1 0.2487] Hide on client
        (Development|SetHiddeninGame (Variables|Default|GetBall) true false)))))

(fn ConstructionScript
  (|CallParentFunction))
